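# Editor utility widget that lays out Content Browser selections as a "runway" of evenly spaced level actors, optionally wrapping into rows.
# One button spawns selected assets, excluding random color prop palettes; another spawns a wall, floor or roof blockout per selected BuildingTextureData and applies it.
# Counters drive X/Y offsets from a starting position; bounds-based spacing is disabled.
# Username checks reveal texture-data controls and a Python-driven button that opens the widget's editor.
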EditorUtilityWidgetBlueprint /Game/Environments/Blueprints/RunwayGenerator/EUW_Runway_Generator
graphs [
 EventGraph {nodes [
   "On Clicked (BTN_SpawnTDActors)"!() -> (OutputDelegate, then, widget:User Widget Object Reference, event:Pointer Event Structure) {is=title K2Node_ComponentBoundEvent}
   > n1 = GetSelectedTDs!(execute, self) -> (then, AR_TextureData:Array of Building Texture Data Object References)
   > Create Transaction!(Begin:Exec, Finalize:Exec=@n2.Completed, Cancel:Exec, Context:String, Description:Text) -> (Pending:Exec, Complete:Exec) {is=title K2Node_MacroInstance}
   > n2 = For Each Loop(Array:Array of Building Texture Data Object References=@n1) -> (Array Element:Building Texture Data Object Reference)
   > Set TextureData(TextureData:Building Texture Data Object Reference=@n2) -> (Output_Get:Building Texture Data Object Reference)
   > SetXAxisOffset!(execute, self, Actor:"Object Reference (by ref)") -> (then)
   > SpawnActorInRunway!(execute, self, ObjectToUse:Object Reference=Get ObjectToSpawn() -> (ObjectToSpawn:Object Reference)) -> (then)
   > SetYAxisOffset()
   > n3 = Cast To BuildingSMActor!(execute, Object=Get PreviouslySpawnedActor() -> (PreviouslySpawnedActor:Actor Object Reference)) -> (then, CastFailed, AsBuilding Static Mesh:Building Static Mesh Object Reference) {is=title K2Node_DynamicCast}
   n4 = Get TextureData() -> (TextureData:Building Texture Data Object Reference)
   > SetBuildingTextureData!(execute, self:Building Static Mesh Object Reference=@n3, Index:int=0, NewBuildingTextureData:Building Texture Data Object Reference=@n4) -> (then)
   > SetBuildingTextureData(self=@n3, Index=1, NewBuildingTextureData=@n4)
   "On Property Changed (DV_ObjectSelect)"!() -> (OutputDelegate, then, PropertyName:Name=None) {is=title K2Node_ComponentBoundEvent}
   > n5 = Switch on ENUM_RG_EnvAssetType!(execute, Selection:ENUM_RG_EnvAssetType Enum=Get SelectedAssetType() -> (SelectedAssetType:ENUM_RG_EnvAssetType Enum=NewEnumerator0)) -> (NewEnumerator0:Exec, NewEnumerator1:Exec, NewEnumerator2:Exec) {is=title K2Node_SwitchEnum}
   > Set ObjectToSpawn(ObjectToSpawn:Object Reference=/Game/Environments/Blockout/Blueprints/Walls/BP_Blockout_Wall_A_A) -> (Output_Get:Object Reference)
   Set ObjectToSpawn(execute=@n5.NewEnumerator1, ObjectToSpawn=/Game/Environments/Blockout/Blueprints/Floors/BP_Blockout_Floor_A_A)
   Set ObjectToSpawn(execute=@n5.NewEnumerator2, ObjectToSpawn=/Game/Environments/Blockout/Blueprints/Roof/BP_Blockout_Roof_A_A)
   "On Clicked (BTN_EditWidget)"!() -> (OutputDelegate, then, widget:User Widget Object Reference, event:Pointer Event Structure) {is=title K2Node_ComponentBoundEvent}
   > n6 = GetBlueprintAsset(Object=&self)
   > Execute Python Script!(
    execute
    PythonScript="import unreal\r\n\r\nmy_widget = widget_object\r\nasset_editor_subsystem = unreal.get_editor_subsystem(unreal.AssetEditorSubsystem)\r\nasset_editor_subsystem.open_editor_for_assets([my_widget])"
    widget_object:Blueprint Object Reference=@n6
   ) -> (then, ReturnValue) {inputs [widget_object]}
   "On Clicked (BTN_SpawnActors)"!() -> (OutputDelegate, then, widget:User Widget Object Reference, event:Pointer Event Structure) {is=title K2Node_ComponentBoundEvent}
   > GetSelectedAssets!(execute, self:Self Object Reference) -> (then)
   > GetActorBounds()
   > Create Transaction(Finalize=@n7.Completed)
   > n7 = For Each Loop(Array:Array of Object References=Get FilteredAssets() -> (FilteredAssets:Array of Object References)) -> (Array Element:Object Reference)
   > SetXAxisOffset(Actor=@n7)
   > SpawnActorInRunway(ObjectToUse=@n7)
   > SetYAxisOffset()
   Event PreConstruct()
   > n8 = Branch(Condition="Get SortBySize (Disabled)"() -> ("SortBySize (Disabled)":bool))
   > n9 = "Set SortBySize (Disabled)"("SortBySize (Disabled)":bool=false) -> (Output_Get:bool)
   n10 = Self-Reference()
   SetObject(
    execute=[@n8.else, @n9.then]
    self=Get DV_ActorPlacementOptions() -> (DV_ActorPlacementOptions:Details View Object Reference)
    NewObject=@n10
   )
   > SetObject(self=Get DV_RowOptions() -> (DV_RowOptions:Details View Object Reference), NewObject=@n10)
   > SetObject(self=Get DV_ObjectSelect() -> (DV_ObjectSelect:Details View Object Reference), NewObject=Self-Reference())
   > Sequence()
   > n11 = Branch(Condition=Get SpawnMultipleRows() -> (SpawnMultipleRows:bool))
   n12 = Get DV_RowOptions()
   > SetVisibility(self=@n12, InVisibility=Visible)
   SetVisibility(execute=@n11.else, self=@n12, InVisibility=Hidden)
   Event Construct()
   > Set SelectedAssetType(SelectedAssetType:ENUM_RG_EnvAssetType Enum=NewEnumerator0) -> (Output_Get:ENUM_RG_EnvAssetType Enum=NewEnumerator0)
   > n13 = Sequence()
   > n14 = Branch(Condition="Equal Exactly (String)"(A=GetPlatformUserName!() -> (ReturnValue:String), B=brianfasten))
   n15 = Get BTN_EditWidget() -> (BTN_EditWidget:TA Generic 01 Button Object Reference)
   > SetVisibility(self=@n15, InVisibility=Visible)
   SetVisibility(execute=@n14.else, self=@n15, InVisibility=Hidden)
   n16 = Branch(execute=@n13.then_1, Condition="Equal Exactly (String)"(A=GetPlatformUserName(), B=eanvasquez))
   n17 = Get DV_ObjectSelect()
   n18 = Get BTN_SpawnTDActors() -> (BTN_SpawnTDActors:TA Generic 01 Button Object Reference)
   > SetVisibility(self=[@n17, @n18], InVisibility=Visible)
   SetVisibility(execute=@n16.else, self=[@n17, @n18], InVisibility=Hidden)
  ]}
 GetActorBounds {nodes [
   GetActorBounds() {is=title K2Node_FunctionEntry}
   > Clear!(execute, TargetMap:Map of Object References to Vectors=Get MAP_OriginBounds() -> (MAP_OriginBounds:Map of Object References to Vectors)) -> (then) {is=title K2Node_CallFunction}
   > Branch(Condition="Get SortBySize (Disabled)"())
   > n1 = For Each Loop(Array=Get FilteredAssets())
   > n2 = SpawnActorFromObject!(execute, ObjectToUse:Object Reference=@n1, Location:Vector="0, 0, 0", Rotation:Rotator, bTransient:bool=false) -> (then, ReturnValue:Actor Object Reference)
   > n3 = Set L_BoundActors(L_BoundActors:Actor Object Reference=@n2) -> (Output_Get:Actor Object Reference)
   > n4 = Branch(Condition=false)
   n5 = Get Actor Bounds(self=@n3, bOnlyCollidingComponents=false)
   > Add!(
    execute
    TargetMap:Map of Object References to Vectors=Get MAP_OriginBounds()
    Key:"Object Reference (by ref)"=@n3
    Value:"Vector (by ref)"=@n5.BoxExtent
   ) -> (then) {is=title K2Node_CallFunction}
   PrintString(execute=@n4.else, InString=Append(A=SPAWNACTORBOUNDS_, B="To String (Vector)"!(InVec:Vector=@n5) -> (ReturnValue:String)))
   > Add(TargetMap=Get MAP_OriginBounds(), Key=@n3, Value=@n5)
   Destroy Actor()
   Reroute Node(InputPin:Actor Object Reference=@n3) -> (OutputPin:Actor Object Reference)
  ]}
 GetLocationToPlaceActor {nodes [
   GetLocationToPlaceActor() {is=title K2Node_FunctionEntry}
   n1 = Get PreviouslySpawnedActor()
   > Set SpawnedActorPosition(SpawnedActorPosition:Vector=Get Actor Location(self=@n1)) -> (Output_Get:Vector)
   n2 = Get Actor Bounds(self=@n1, bOnlyCollidingComponents=false)
   > Set OriginBound_X(OriginBound_X:double=float / float(A=BreakVector(InVec=@n2), B=2)) -> (Output_Get:double)
  ]}
 GetSelectedAssets {nodes [
   GetSelectedAssets() {is=title K2Node_FunctionEntry}
   > Clear(TargetArray:Array of Object References=Get SelectedAssets() -> (SelectedAssets:Array of Object References)) {is=title K2Node_CallArrayFunction}
   > Clear(TargetArray=Get FilteredAssets())
   > n1 = GetSelectedAssets!(execute) -> (then, ReturnValue:Array of Object References) {is=title K2Node_CallFunction}
   > n2 = Filter by Class!(
    execute
    TargetArray:Array of Object References=@n1
    ObjectClass:Object Class Reference=/Script/FortniteGame.FortRandomColorPropPalette
    FilterType:EEditorScriptingFilterType Enum=Exclude
   ) -> (then, ReturnValue:Array of Fort Random Color Prop Palette Object References)
   > Set FilteredAssets(FilteredAssets:Array of Object References=@n2) -> (Output_Get:Array of Object References)
  ]}
 GetSelectedTDs {nodes [
   GetSelectedTDs() {is=title K2Node_FunctionEntry}
   > n1 = GetSelectedAssets()
   > n2 = Filter by Class(TargetArray=@n1, ObjectClass=/Script/FortniteGame.BuildingTextureData) -> (ReturnValue:Array of Building Texture Data Object References)
   > n3 = Set L_AR_TextureData(L_AR_TextureData:Array of Building Texture Data Object References=@n2) -> (Output_Get:Array of Building Texture Data Object References)
   > Return Node!(execute, AR_TextureData:Array of Building Texture Data Object References=@n3)
  ]}
 Get_SB_SpawnOffset_Value {nodes [
   Get_SB_SpawnOffset_Value() {is=title K2Node_FunctionEntry}
   > Return Node!(execute, ReturnValue:float="To Float (Integer)"(InInt=Get SpawnDistance() -> (SpawnDistance:int=0)))
  ]}
 SetXAxisOffset {nodes [
   n1 = SetXAxisOffset!() -> (then, Actor:"Object Reference (by ref)") {is=title K2Node_FunctionEntry}
   > n2 = Branch(Condition="Get SortBySize (Disabled)"())
   n3 = Find(TargetMap:Map of Object References to Vectors=Get MAP_OriginBounds(), Key:"Object Reference (by ref)"=@n1) -> (Value:Vector)
   > PrintString(InString=Append(A=BOUNDSVALUE_, B="To String (Vector)"(InVec=@n3)))
   > n4 = Set OriginBound_X(OriginBound_X=BreakVector(InVec=@n3))
   n5 = Get SpawnCounter() -> (SpawnCounter:int=0)
   n6 = Get SpawnDistance() -> (SpawnDistance=0)
   n7 = float * float(A:int=@n5, B:int=@n6) {is=title K2Node_PromotableOperator}
   > Set XAxisOffsetAmount(XAxisOffsetAmount:int=Truncate(A=float + float(A=@n4, B=@n7))) -> (Output_Get:int=0)
   Set XAxisOffsetAmount(execute=@n2.else, XAxisOffsetAmount=int * int(A=@n5, B=@n6)) -> (Output_Get=0)
  ]}
 SetYAxisOffset {nodes [
   SetYAxisOffset() {is=title K2Node_FunctionEntry}
   > Branch(Condition=Get SpawnMultipleRows())
   n1 = Get NumberOfActorsPerRow() -> (NumberOfActorsPerRow:int=0)
   > Branch(Condition="integer > integer"(A=Get SpawnCounter() -> (SpawnCounter=0), B=int - int(A=@n1)))
   > Set SpawnCounter(SpawnCounter:int=0) -> (Output_Get:int=0)
   > n2 = Set RowOffset(RowOffset:int=int + int(A=Get RowOffset() -> (RowOffset:int=0))) -> (Output_Get:int=0)
   > Set YAxisOffsetAmount(YAxisOffsetAmount:int=int * int(A=@n2, B=Get SpawnDistance() -> (SpawnDistance=0))) -> (Output_Get:int=0)
  ]}
 SpawnActorInRunway {nodes [
   n1 = SpawnActorInRunway!() -> (then, ObjectToUse:Object Reference) {is=title K2Node_FunctionEntry}
   n2 = BreakVector(InVec=Get StartingPosition() -> (StartingPosition:Vector))
   n3 = Get YAxisOffsetAmount() -> (YAxisOffsetAmount:int=0)
   n4 = Get XAxisOffsetAmount() -> (XAxisOffsetAmount:int=0)
   > n5 = SpawnActorFromObject(
    ObjectToUse=@n1
    Location=MakeVector(X=float + float(A:int=@n4, B=@n2), Y=float + float(A:double=@n2.Y, B:int=@n3), Z=@n2.Z)
    bTransient=false
   )
   > Set PreviouslySpawnedActor(PreviouslySpawnedActor:Actor Object Reference=@n5) -> (Output_Get:Actor Object Reference)
   > Set SpawnCounter(SpawnCounter=int + int(A=Get SpawnCounter() -> (SpawnCounter=0))) -> (Output_Get=0)
  ]}
]
variables [
 Actor {type=Object, default=null}
 AR_ExcludedObjectClasses {type="array<Class>", default [&FortRandomColorPropPalette]}
 BoxExtentBounds {type=Vector}
 CollectionList {type="array<string>", default []}
 FilteredAssets {type="array<Object>", default []}
 MAP_BoxExtentBounds {default {}}
 MAP_OriginBounds {default {}}
 NumberOfActorsPerRow {type=integer, default=10}
 ObjectToSpawn {type=Object, default=&BP_Blockout_Wall_A_A}
 OriginBound_X {type=number}
 OriginBounds {type=Vector}
 PreviouslySpawnedActor {type=Actor, default=null}
 RowOffset {type=integer}
 SelectedAssets {type="array<Object>", default []}
 SelectedAssetType {type=ENUM_RG_EnvAssetType, default=Wall}
 "SortBySize (Disabled)" {type=boolean}
 SpawnCounter {type=integer}
 SpawnDistance {type=integer, default=512}
 SpawnedActorPosition {type=Vector}
 SpawnMultipleRows {type=boolean}
 StartingPosition {type=Vector, default=<-12288 -12288 0>}
 TextureData {type=BuildingTextureData, default=null}
 XAxisOffsetAmount {type=integer}
 YAxisOffsetAmount {type=integer}
]
widgets {
 info {rootWidgetClass=&HorizontalBox, widgetCount=9}
 widgets [
  HorizontalBox_0 {widgetClassPath=&HorizontalBox, widget=HorizontalBox_0}
  VerticalBox_0 {
   widgetClassPath=&VerticalBox
   parent=HorizontalBox_0
   widget=VerticalBox_0
   properties {toolTipText="Takes user-selection from content browser and lays them out in a straight line starting from world origin (0,0,0)."}
   slot {class=HorizontalBoxSlot, properties {size {sizeRule=Fill, value=1}}}
  }
  DV_ActorPlacementOptions {
   widgetClassPath=&DetailsView
   parent=VerticalBox_0
   widget=DV_ActorPlacementOptions
   properties {categoriesToShow [Actor Placement Options]}
   slot {class=VerticalBoxSlot}
  }
  DV_RowOptions {
   widgetClassPath=&DetailsView
   parent=VerticalBox_0
   widget=DV_RowOptions
   properties {categoriesToShow [Row Options]}
   slot {class=VerticalBoxSlot}
  }
  BTN_SpawnActors {
   widgetClassPath=&TA_Generic01_Button
   parent=VerticalBox_0
   widget=BTN_SpawnActors
   properties {label=Make Runway From Selection, toolTipText="Takes user-selection from content browser and lays them out in a single, straight line"}
   slot {class=VerticalBoxSlot, properties {padding=<ltrb 10 10 10 10>}}
  }
  VerticalBox_1 {widgetClassPath=&VerticalBox, parent=VerticalBox_0, widget=VerticalBox_1, slot {class=VerticalBoxSlot}}
  DV_ObjectSelect {
   widgetClassPath=&DetailsView
   parent=VerticalBox_1
   widget=DV_ObjectSelect
   properties {categoriesToShow [TextureData]}
   slot {class=VerticalBoxSlot, properties {padding=<ltrb 4 4 4 4>}}
  }
  BTN_SpawnTDActors {
   widgetClassPath=&TA_Generic01_Button
   parent=VerticalBox_1
   widget=BTN_SpawnTDActors
   properties {label=Make Texture Data Runways, toolTipText="Takes user-selection from content browser and lays them out in a single, straight line"}
   slot {class=VerticalBoxSlot}
  }
  BTN_EditWidget {
   widgetClassPath=&TA_Generic01_Button
   parent=HorizontalBox_0
   widget=BTN_EditWidget
   properties {label=">>", buttonColor_Default=<rgba 0.0469 0.0469 0.0469 1>, sizeOverride=<20 0>}
   slot {class=HorizontalBoxSlot, properties {padding=<ltrb 5 5 5 5>}}
  }
 ]
}
